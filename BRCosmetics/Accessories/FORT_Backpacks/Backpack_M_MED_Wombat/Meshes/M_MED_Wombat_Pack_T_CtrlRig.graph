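# Control Rig for the M_MED_Wombat backpack cosmetic that shifts the pack bone's position based on its tilt.
# It reads the pack's rotation relative to root as ZYX Euler angles and feeds X into two clamped remaps.
# These drive Z and Y translation offsets, applied to pack in additive local mode.
# The dyn bolt child bones are in the hierarchy, but the graph does not modify them directly.
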ControlRigBlueprint /BRCosmetics/Accessories/FORT_Backpacks/Backpack_M_MED_Wombat/Meshes/M_MED_Wombat_Pack_T_CtrlRig
graphs [
 RigVMFunctionLibrary {links [], nodes []}
 RigVMModel {nodes [
   RigUnit_BeginExecution = &RigUnit_BeginExecution()
   n1 = &RigVMFunction_MathQuaternionToEuler(RotationOrder=ZYX, Value=&RigUnit_GetRelativeBoneTransform(Bone=pack, Space=root).Transform.Rotation)
   /Script/ControlRig.RigUnit_ModifyBoneTransforms(
    BoneToModify=((Bone="pack",Transform=(Rotation=(X=0,Y=0,Z=0,W=1),Translation=(X=0,Y=0,Z=0),Scale3D=(X=1,Y=1,Z=1))))
    Mode=AdditiveLocal
    Weight=1
    WeightMaximum=1
    BoneToModify.0.Transform.Translation.Z=&RigVMFunction_MathFloatRemap(SourceMaximum=-15, SourceMinimum=-7, TargetMaximum=-1.5, TargetMinimum=-3, bClamp=true, Value=@n1.Result.X).Result
    BoneToModify.0.Transform.Translation.Y=&RigVMFunction_MathFloatRemap(SourceMaximum=-15, SourceMinimum=-7, TargetMaximum=0, TargetMinimum=0.5, bClamp=true, Value=@n1.Result.X).Result
    ExecutePin=@RigUnit_BeginExecution.ExecutePin
   ) {is=struct RigVMUnitNode}
  ]}
]
hierarchy [
 root {loc=<0 0 0>, rot=<0 0 0 1>}
 pack {parent=root, loc=<0 -0.7316 0>, rot=<0 0 0 1>}
 dyn_Bolt_MSTR {parent=pack, loc=<0 -1.7441 11.5383>, rot=<0 0 0 1>}
 dyn_bolt_03 {parent=dyn_Bolt_MSTR, loc=<7.1937 2.5314 -2.7733>, rot=<0 0 0 1>}
 dyn_bolt_02 {parent=dyn_Bolt_MSTR, loc=<7.6077 2.7413 -0.0854>, rot=<0 0 0 1>}
 dyn_bolt_01 {parent=dyn_Bolt_MSTR, loc=<7.1938 3.0028 2.9545>, rot=<0 0 0 1>}
]
