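# Animation Blueprint for the M_MED_Wombat backpack cosmetic that adjusts the pack bone to follow spine bend.
# Each update maps the owner mesh's spine_05-to-spine_01 delta yaw from ref pose into a clamped, inverse roll in backpack_rotate.
# The output pose adds that rotation and a small fixed offset to pack on the ref pose.
# A Control Rig chain with targeting blend and crouch offset is not connected to the output.
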AnimBlueprint /BRCosmetics/Accessories/FORT_Backpacks/Backpack_M_MED_Wombat/Meshes/M_MED_Wombat_Pack_AnimBP
target_skeleton=&/M_MED_Wombat_Pack_Skeleton
graphs [
 AnimGraph {nodes [
   n1 = "Transform (Modify) Bone - Bone: pack"(ComponentPose=Local To Component(LocalPose=Local Space Ref Pose()), Translation="0,0,-0.5", Rotation=Get backpack_rotate()) {
    rotationMode=BMM_Additive
    rotationSpace=BCS_BoneSpace
    translationMode=BMM_Additive
    alphaBoolBlend {blendOption=Linear}
    boneToModify {boneName=pack}
    translation {z=-0.5}
   }
   n2 = Control Rig(Source=Component To Local(ComponentPose=@n1)) {
    alphaBoolBlend {blendOption=Linear}
    controlRigAssetReference {blueprintRigClass=&/M_MED_Wombat_Pack_T_CtrlRig.M_MED_Wombat_Pack_T_CtrlRig_C}
   }
   n3 = Control Rig() {
    alphaBoolBlend {blendOption=Linear}
    controlRigAssetReference {blueprintRigClass=&/M_MED_Wombat_Pack_CtrlRig.M_MED_Wombat_Pack_CtrlRig_C}
   }
   n4 = Get bIsTargeting()
   n5 = "Transform (Modify) Bone - Bone: pack"!(
    ComponentPose=Local To Component(LocalPose=Blend Poses by bool(BlendPose_0=@n2, BlendPose_1=@n3, bActiveValue=@n4) {blendType=Linear})
    bAlphaBoolEnabled:bool=Get bIsCrouching()
    Translation="0,0,1"
    Rotation="0,0,2.5"
    Scale
   ) -> (Pose) {
    alphaInputType=Bool
    rotationMode=BMM_Additive
    rotationSpace=BCS_BoneSpace
    translationMode=BMM_Additive
    alphaBoolBlend {blendInTime=0.1, blendOption=Linear, blendOutTime=0.1}
    boneToModify {boneName=pack}
    rotation {roll=2.5}
    translation {z=1}
   }
   Component To Local(ComponentPose=@n5)
   n6 = "Transform (Modify) Bone - Bone: pack"!(
    ComponentPose=Local To Component(LocalPose=Local Space Ref Pose())
    Alpha:float
    Translation="0,-1.7,0"
    Rotation=Get backpack_rotate()
    Scale
   ) -> (Pose) {
    rotationMode=BMM_Additive
    rotationSpace=BCS_BoneSpace
    translationMode=BMM_Additive
    alphaBoolBlend {blendOption=Linear}
    boneToModify {boneName=pack}
    translation {y=-1.7}
   }
   Output Pose(Result=Component To Local(ComponentPose=@n6)) {layerGroup=None}
  ]}
 EventGraph {nodes [
   Event BlueprintUpdateAnimation()
   n1 = GetOwnerSkeletalMeshComponent()
   n2 = BreakRotator(InRot=BreakTransform(InTransform=GetDeltaTransformFromRefPose(self=@n1, BoneName=spine_05, BaseName=spine_01)).Rotation)
   > Set backpack_rotate(backpack_rotate=MakeRotator(Roll=MapRangeClamped(Value=@n2.Yaw, InRangeA=5, InRangeB=28, OutRangeA=15, OutRangeB=5)))
   TryGetPawnOwner()
  ]}
]
variables [backpack_rotate {type=Rotator}]
